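# Physics asset for the Ashton Salt Lake face accessory cosmetic, defining physics bodies for its dynamic hair bones.
# Small capsules on the front, middle and back hair bones sit alongside a larger head capsule.
# Each hair body is constrained directly to the head rather than chained to the others, with matching swing and twist limit angles on all three joints.
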PhysicsAsset /BRCosmetics/Accessories/Hats/M_MED_AshtonSaltLake_FaceAcc/Meshes/M_MED_Ashton_Salt_Lake_FaceAcc_Dyn
bodies {
 dyn_hair_bk [{shapeType=Capsule, length=2, radius=2, center=<0.15 0 0>, rotation=<rot 90 0 0>}]
 dyn_hair_fr [{shapeType=Capsule, length=2, radius=2, center=<0.2 0 0>, rotation=<rot 90 0 0>}]
 dyn_hair_mid [{shapeType=Capsule, length=2, radius=2, center=<0.1 0 0>, rotation=<rot 90 0 0>}]
 head [{shapeType=Capsule, length=3.5141, radius=3.6667, rotation=<rot 90 0 0>}]
}
constraints [
 {bone1Name=dyn_hair_fr, bone2Name=head, swing1LimitDegrees=45, swing2LimitDegrees=45, twistLimitDegrees=45}
 {bone1Name=dyn_hair_mid, bone2Name=head, swing1LimitDegrees=45, swing2LimitDegrees=45, twistLimitDegrees=45}
 {bone1Name=dyn_hair_bk, bone2Name=head, swing1LimitDegrees=45, swing2LimitDegrees=45, twistLimitDegrees=45}
]
